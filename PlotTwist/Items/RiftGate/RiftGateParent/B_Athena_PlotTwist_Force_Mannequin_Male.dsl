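; Male Force mannequin skeletal mesh actor for PlotTwist Rift Gate items that plays montages on request.
; Async-loads soft-referenced idle and current montages, with stop and play-idle events.
; An optional timer returns to idle after montage length minus the supplied position, though playback itself starts at zero.
; On cosmetic-capable BeginPlay it caches budgeted child skeletal meshes, which a helper function uses to set scalar material parameters.
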(event EventBeginPlay
  ;[1 1 1] Begin Play
  (|CallParentFunction)
  (Networking|CanExecuteCosmeticEvents
    (:True
      (CallFunction|CacheSkeletalMeshComponentReferences))))

(event Custom|SetIdleMontage
  (params (Animation%20Montage :type Anim%20Montage%20Soft%20Object%20Reference))
  ;[1 1 1] Set Idle Montage
  (bind _object (Utilities|AsyncLoadAsset Animation%20Montage)
    (:Completed
      (bind _asanim_montage (Utilities|Casting|CastToAnimMontage _object))
      (Utilities|IsValid _asanim_montage
        (:Is%20Valid
          (Variables|Default|SetIdleMontage _asanim_montage))))))

(event Custom|PlayIdleMontage
  ;[1 1 1] Play Idle Montage
  (Utilities|Time|ClearandInvalidateTimerbyHandle (Variables|Default|GetTimerHandle_PlayIdleMontage))
  (bind _idlemontage (Variables|Default|GetIdleMontage)
    (:then
      (Animation|Montage|MontagePlay (Components|SkeletalMesh|GetAnimInstance (Variables|SkeletalMeshActor|GetSkeletalMeshComponent)) _idlemontage 1.0 "MontageLength" 0.0 true))
    (:else)))

(event Custom|StopMontage
  ;[1 1 1] Stop Montage
  (Components|Animation|Stop (Variables|SkeletalMeshActor|GetSkeletalMeshComponent)))

(event Custom|LoadAndPlayCurrentMontage
  (params (Animation%20Montage :type Anim%20Montage%20Soft%20Object%20Reference) (AnimCurrentPosition :type Float%20%28double-precision%29) (ReturnToIdle :type Boolean))
  ;[1 1 1] Load and Play Current Montage
  (Utilities|Time|ClearandInvalidateTimerbyHandle (Variables|Default|GetTimerHandle_PlayIdleMontage))
  (Variables|Default|SetReturn%20to%20Idle ReturnToIdle)
  (bind _object (Utilities|AsyncLoadAsset Animation%20Montage)
    (:Completed
      (bind _asanim_montage (Utilities|Casting|CastToAnimMontage _object))
      (Utilities|IsValid _asanim_montage
        (:Is%20Valid
          (Variables|Default|SetCurrentMontage _asanim_montage)
          (Variables|Default|SetMontageStartingPosition AnimCurrentPosition)
          (bind _skeletalmeshcomponent (Variables|SkeletalMeshActor|GetSkeletalMeshComponent)
            (:then
              (Animation|Montage|MontagePlay (Components|SkeletalMesh|GetAnimInstance _skeletalmeshcomponent) (Variables|Default|GetCurrentMontage) 1.0 "MontageLength" 0.0 true)
              (if (Variables|Default|GetReturn%20to%20Idle)
                (bind _returnvalue (Utilities|Time|SetTimerbyEvent (EventDispatchers|CreateEvent :!event Custom|PlayIdleMontage) (- (Animation|GetPlayLength (Variables|Default|GetCurrentMontage)) (Variables|Default|GetMontageStartingPosition))))
                (Variables|Default|SetTimerHandle_PlayIdleMontage _returnvalue)))
            (:else)))))))

(fn CacheSkeletalMeshComponentReferences
  (Utilities|Array|Clear (Variables|Default|GetSkeletalMeshComponents))
  (for _array_element (Components|GetChildrenComponents (Variables|SkeletalMeshActor|GetSkeletalMeshComponent) true)
    (bind _asskeletal_mesh_component_budgeted (Utilities|Casting|CastToSkeletalMeshComponentBudgeted _array_element))
    (Utilities|Array|Add (Variables|Default|GetSkeletalMeshComponents) _asskeletal_mesh_component_budgeted)))

(fn SetMaterialParamaterForCachedSkeletalMeshes
  (params (MaterialParameterName :type Name) (NewValue :type Float%20%28double-precision%29))
  (bind _skeletalmeshcomponents (Variables|Default|GetSkeletalMeshComponents))
  (if (not (Utilities|Array|IsEmpty _skeletalmeshcomponents))
    (for _array_element _skeletalmeshcomponents
      (Rendering|Material|SetScalarParameterValueonMaterials _array_element (Variables|GetMaterialParameterName) (Variables|GetNewValue)))))

(fn ConstructionScript
  (|CallParentFunction))
